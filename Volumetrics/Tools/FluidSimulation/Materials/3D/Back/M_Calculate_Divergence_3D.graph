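# Unlit material that calculates the divergence of a 3D velocity field for the fluid simulation tool.
# VolumeOffsetSamples reads neighboring voxels along each axis from the Velocity texture, a volume laid out as XY frames.
# Central differences of the matching X, Y and Z components are summed and scaled by half the inverse GridScale.
# Negative emissive color is allowed, so signed values are possible.
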Material /Volumetrics/Tools/FluidSimulation/Materials/3D/Back/M_Calculate_Divergence_3D
graph {
 properties {bAllowNegativeEmissiveColor=true, pixelDepthOffsetMode=PDOM_Legacy, shadingModel=MSM_Unlit, usedShadingModels=Unlit}
 nodes [
  n1 = MaterialFunctionCall!(
   Texture Object=TextureObjectParameter() {desc=Velocity, parameterName=Velocity, texture=/Volumetrics/Tools/FluidSimulation/Textures/RT_Tiny}
   Resolution=VectorParameter() {parameterName=Resolution, defaultValue=<rgba 4096 4096 0 0>}.RGB
   XYFrames=VectorParameter() {parameterName=XYFrames, defaultValue=<rgba 16 16 0 0>}.RGB
   Clamp Edges
  ) {materialFunction=/Volumetrics/Tools/FluidSimulation/MaterialFunctions/VolumeOffsetSamples}
  n2 = Subtract(A=ComponentMask(None=@n1.+X) {r=true}, B=ComponentMask(None=@n1.-X) {r=true})
  n3 = Subtract(A=ComponentMask(None=@n1.+Y) {g=true}, B=ComponentMask(None=@n1.-Y) {g=true})
  n4 = ComponentMask(None=@n1.+Z) {b=true}
  n5 = ComponentMask(None=@n1.-Z) {b=true}
  Multiply(
   A=Divide(B=ScalarParameter() {defaultValue=1, parameterName=GridScale}) {constA=0.5, desc=half rdx}
   B=Add(A=Add(A=@n2, B=@n3), B=Subtract(A=@n4, B=@n5))
  )
 ]
}
