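# Unlit material for the gradient subtraction pass of a 3D fluid simulation, correcting velocity by the pressure gradient.
# VolumeOffsetSamples reads six axis neighbours from the pressure volume, stored as a 2D frame atlas, with clamped edges.
# Their red-channel differences form a gradient, scaled by half the inverse grid scale and subtracted from the centre velocity sample.
# Alpha-composite blending with negative emissive allowed suits render-target passes carrying signed values.
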Material /Volumetrics/Tools/FluidSimulation/Materials/3D/Back/M_Gradient_Subtraction_3D
graph {
 properties {
  bAllowNegativeEmissiveColor=true
  blendMode=BLEND_AlphaComposite
  pixelDepthOffsetMode=PDOM_Legacy
  shadingModel=MSM_Unlit
  usedShadingModels=Unlit
 }
 nodes [
  n1 = TextureSampleParameter2D(UVs=TextureCoordinate()) {desc=Velocity - Center Sample, parameterName=Velocity, texture=/Volumetrics/Tools/FluidSimulation/Textures/RT_Tiny}
  OneMinus(None=@n1.A)
  n2 = Divide(B=ScalarParameter() {defaultValue=1, parameterName=GridScale}) {constA=0.5, desc=half rdx}
  n3 = MaterialFunctionCall!(
   Texture Object=TextureObjectParameter() {desc=Velocity, parameterName=Pressure, texture=&RT_Tiny}
   Resolution=VectorParameter() {parameterName=Resolution, defaultValue=<rgba 4096 4096 0 0>}.RGB
   XYFrames=VectorParameter() {parameterName=XYFrames, defaultValue=<rgba 16 16 0 0>}.RGB
   Clamp Edges
  ) {materialFunction=/Volumetrics/Tools/FluidSimulation/MaterialFunctions/VolumeOffsetSamples}
  n4 = MaterialFunctionCall(
   X=Subtract(A=ComponentMask(None=@n3.+X) {r=true}, B=ComponentMask(None=@n3.-X) {r=true})
   Y=Subtract(A=ComponentMask(None=@n3.+Y) {r=true}, B=ComponentMask(None=@n3.-Y) {r=true})
   Z=Subtract(A=ComponentMask(None=@n3.+Z) {r=true}, B=ComponentMask(None=@n3.-Z) {r=true})
  ) {materialFunction=&MakeFloat3}
  Subtract(A=@n1.RGB, B=Multiply(A=Multiply(A=@n2, B=@n4.Result)))
 ]
}
