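# Unlit material that computes one Jacobi-style pressure solve iteration for a 3D fluid simulation on volumes stored as 2D frame atlases.
# It sums six pressure neighbors from VolumeOffsetSamples, subtracts divergence times GridScale squared, then divides by NumSamples.
# Divisor scales the sampling resolution and blends Z neighbors toward the center value, and negative emissive output is allowed.
# Mask-based neighbor lerps and a second divergence sample sit unconnected to that calculation.
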Material /Volumetrics/Tools/FluidSimulation/Materials/3D/Back/M_Pressure_Iteration_3D
graph {
 properties {bAllowNegativeEmissiveColor=true, pixelDepthOffsetMode=PDOM_Legacy, shadingModel=MSM_Unlit, usedShadingModels=Unlit}
 nodes [
  n1 = ScalarParameter() {defaultValue=1, parameterName=GridScale}
  n2 = Multiply(A=Multiply(A=@n1, B=@n1)) {constB=-1}
  n3 = VectorParameter() {parameterName=XYFrames, defaultValue=<rgba 16 16 0 0>}
  n4 = MaterialFunctionCall!(
   Texture Object=TextureObjectParameter() {parameterName=Divergence, texture=/Volumetrics/Tools/FluidSimulation/Textures/RT_Tiny}
   Position=MaterialFunctionCall!(UVs, XY Frames=ComponentMask(None=@n3.RGB) {g=true, r=true}) {materialFunction=/Volumetrics/Content/VolumeTextures/MaterialFunctions/EncodeVolumeCoordinates}.Result
   Scale=Constant() {r=1}
   Num Frames=ComponentMask(None=VectorParameter() {parameterName=SampleXYFrames, defaultValue=<rgba 16 16 0 0>}.RGB) {g=true, r=true}
  ) {materialFunction=/Volumetrics/Content/VolumeTextures/MaterialFunctions/VolumeTextureFunction_NoZBlend}
  n5 = ScalarParameter() {defaultValue=1, parameterName=Divisor}
  n6 = MaterialFunctionCall!(
   Texture Object=TextureObjectParameter() {desc=Velocity, parameterName=Pressure, texture=&RT_Tiny}
   Resolution=Multiply(A=VectorParameter() {parameterName=Resolution, defaultValue=<rgba 4096 4096 0 0>}.RGB, B=@n5)
   XYFrames=VectorParameter() {parameterName=XYFrames, defaultValue=<rgba 16 16 0 0>}.RGB
   Clamp Edges
  ) {materialFunction=/Volumetrics/Tools/FluidSimulation/MaterialFunctions/VolumeOffsetSamples}
  n7 = Add(
   A=Add(A=ComponentMask(None=@n6.-X) {r=true}, B=ComponentMask(None=@n6.+X) {r=true})
   B=Add(A=ComponentMask(None=@n6.-Y) {r=true}, B=ComponentMask(None=@n6.+Y) {r=true})
  )
  n8 = TextureSampleParameter2D() {desc=Center, parameterName=Pressure, texture=&RT_Tiny}
  n9 = Divide(B=@n5)
  n10 = Add(
   A=LinearInterpolate(A=@n8.R, B=ComponentMask(None=@n6.-Z) {r=true}, Alpha=@n9)
   B=LinearInterpolate(A=@n8.R, B=ComponentMask(None=@n6.+Z) {r=true}, Alpha=@n9)
  )
  Divide(A=Add(A=Multiply(A=@n2, B=@n4.Result), B=Add(A=@n7, B=@n10)), B=ScalarParameter() {defaultValue=6, parameterName=NumSamples})
  n11 = TextureCoordinate()
  n12 = Divide(B=ScalarParameter() {defaultValue=1024, parameterName=Resolution})
  n13 = Multiply(A=@n12, B=Constant2Vector() {r=1})
  LinearInterpolate(Alpha=TextureSampleParameter2D(UVs=Subtract(A=@n11, B=@n13)) {desc=L, parameterName=Mask, texture=/Volumetrics/Content/VolumeTextures/Textures/RT_Default}.B)
  LinearInterpolate(Alpha=TextureSampleParameter2D(UVs=Add(A=@n11, B=@n13)) {desc=R, parameterName=Mask, texture=&RT_Default}.B)
  n14 = TextureCoordinate()
  n15 = Multiply(A=@n12, B=Constant2Vector() {g=1})
  LinearInterpolate(Alpha=TextureSampleParameter2D(UVs=Subtract(A=@n14, B=@n15)) {desc=B, parameterName=Mask, texture=&RT_Default}.B)
  LinearInterpolate(Alpha=TextureSampleParameter2D(UVs=Add(A=@n14, B=@n15)) {desc=T, parameterName=Mask, texture=&RT_Default}.B)
  n16 = TextureCoordinate()
  n17 = VectorParameter() {parameterName=UVScaleAndOffset, defaultValue=<rgba 1 1 0 0>}
  n18 = ComponentMask(None=@n17.RGB) {g=true, r=true}
  TextureSampleParameter2D(UVs=Add(A=Multiply(A=@n16, B=@n18), B=AppendVector(A=@n17.B, B=@n17.B))) {desc=Center, parameterName=Divergence, texture=&RT_Tiny}
 ]
}
